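# Physics asset for the ListDizzy emote gadget prop, defining its simplified collision bodies and joint.
# A large capsule covers the chassis bone and a thin capsule covers tail_01.
# A single constraint joins tail_01 to the chassis with equal, moderate swing and twist limit angles.
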PhysicsAsset /CosmeticCompanions/Assets/Props/Emotes/ListDizzy/Meshes/PA_ListDizzy_Gadget
bodies {
 chassis [{shapeType=Capsule, length=169.0559, radius=46.2501, center=<0.0697 -24.2119 -33.1754>, rotation=<rot 0.8743 -90.0083 -0.0472>}]
 tail_01 [{shapeType=Capsule, length=47.4816, radius=1.808, center=<0.037 -0.0489 -32.7239>, rotation=<rot -0.2326 -90 -0.0194>}]
}
constraints [{bone1Name=tail_01, bone2Name=chassis, swing1LimitDegrees=45, swing2LimitDegrees=45, twistLimitDegrees=45}]
